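# LaserDot is a translucent deferred decal material, apparently the red laser dot for the FloppyNib companion's laser emote.
# An exponential radial gradient is multiplied by itself to tighten the falloff, boosted by a large constant, then tinted red through a Make Vector3 node.
# A SphereMask node is present but unconnected, and the file does not show which material output receives the result.
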Material /CosmeticCompanions/Assets/Quadruped/FloppyNib/Animation/Emotes/FlobbyNib_LaserEmote/FloppyNib_LaserEmote_Player/LaserDot
graph {
 expressions {
  MaterialExpressionConvert_0 {
   class=MaterialExpressionConvert
   inputs [
    None {from=""}
    None {from=""}
    None {from=""}
   ]
   properties {
    nodeName=Make Vector3
    convertInputs [
     {
      type=Scalar
      defaultValue=<rgba 1 0 0 1>
      expressionInput {expression=null, inputName=None, mask=0, maskA=0, maskB=0, maskG=0, maskR=0, outputIndex=0}
     }
     {type=Scalar, expressionInput {expression=null, inputName=None, mask=0, maskA=0, maskB=0, maskG=0, maskR=0, outputIndex=0}}
     {type=Scalar, expressionInput {expression=null, inputName=None, mask=0, maskA=0, maskB=0, maskG=0, maskR=0, outputIndex=0}}
    ]
    convertOutputs [{type=Vector3}]
   }
  }
  MaterialExpressionMaterialFunctionCall_0 {
   class=MaterialExpressionMaterialFunctionCall
   inputs [
    UVs {from=""}
    CenterPosition {from=""}
    Radius {from=""}
    Density {from=""}
    Invert Density {from=""}
   ]
   properties {materialFunction=&RadialGradientExponential}
  }
  MaterialExpressionMultiply_0 {
   class=MaterialExpressionMultiply
   inputs [
    A {from=MaterialExpressionMultiply_2}
    B {from=""}
   ]
   properties {constB=50}
  }
  MaterialExpressionMultiply_1 {
   class=MaterialExpressionMultiply
   inputs [
    A {from=MaterialExpressionMultiply_0}
    B {from=MaterialExpressionConvert_0}
   ]
   properties {constB=500}
  }
  MaterialExpressionMultiply_2 {
   class=MaterialExpressionMultiply
   inputs [
    A {from=MaterialExpressionMaterialFunctionCall_0, output=RadialGradientExponential}
    B {from=MaterialExpressionMaterialFunctionCall_0, output=RadialGradientExponential}
   ]
   properties {constB=50}
  }
  MaterialExpressionSphereMask_0 {
   class=MaterialExpressionSphereMask
   inputs [
    A {from="", output=""}
    B {from="", output=""}
    Radius {from="", output=""}
    Hardness {from="", output=""}
   ]
   properties {}
  }
 }
 properties {bAutomaticallySetUsageInEditor=false, blendMode=BLEND_Translucent, materialDomain=MD_DeferredDecal, usedShadingModels=Default Lit}
}
