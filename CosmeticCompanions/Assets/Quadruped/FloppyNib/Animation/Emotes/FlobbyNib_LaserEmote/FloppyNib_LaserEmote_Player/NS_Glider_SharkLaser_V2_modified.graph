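# Niagara system for the laser in the FloppyNib quadruped companion's laser emote, in a Player folder and named as a modified Glider SharkLaser V2 effect.
# It holds a single emitter named Laser, with no renderers listed and no user variables.
# The only module listed is SystemState, and no data interfaces or dynamic inputs appear.
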NiagaraSystem /CosmeticCompanions/Assets/Quadruped/FloppyNib/Animation/Emotes/FlobbyNib_LaserEmote/FloppyNib_LaserEmote_Player/NS_Glider_SharkLaser_V2_modified
dependencies {usedDataInterfaces [], usedDynamicInputs [], usedModules [&SystemState]}
emitters {Laser {rendererClasses [], renderers []}}
summary {systemName=NS_Glider_SharkLaser_V2_modified, emitters [{emitterName=Laser, rendererClasses []}], userVariables []}
