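# Player animation layer for the FloppyNib laser emote, overriding the Item Override final pose to aim the head and right hand.
# ItemFinalPoseOverride chains Look At controls on head and hand_r (sinusoidal interpolation) into an Apply Additive with no additive pose connected.
# The EventGraph projects AimTarget far along player 0's view direction, but no shown node reads it.
# Other layers and the main AnimGraph hold unconnected nodes.
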AnimBlueprint /CosmeticCompanions/Assets/Quadruped/FloppyNib/Animation/Emotes/FlobbyNib_LaserEmote/FloppyNib_LaserEmote_Player/PlayerLook_AnimLayer
target_skeleton=&Fortnite_M_Avg_Player_Skeleton
graphs [
 AnimGraph {nodes [
   Blend Poses by bool(BlendPose_0=Slot 'InMotion'() {slotName=InMotion}, BlendPose_1=Slot 'DefaultSlot'(), BlendTime_0=0.3, BlendTime_1=0.3)
   Look At(Alpha=1) {boneToModify=~, lookAtTarget=~}
   Output Pose()
   Two Bone IK!(ComponentPose:Component Space Pose Link Structure, Alpha:float=1, EffectorLocation:Vector="0,0,0", JointTargetLocation:Vector="0,0,0") -> (Pose:Component Space Pose Link Structure) {is=title AnimGraphNode_TwoBoneIK}
  ]}
 EventGraph {nodes [
   Event BlueprintUpdateAnimation()
   n1 = GetPlayerViewPoint!(self:Controller Object Reference=GetPlayerController!(PlayerIndex:int=0) -> (ReturnValue:Player Controller Object Reference)) -> (Location:Vector, Rotation:Rotator)
   n2 = GetForwardVector!(InRot:Rotator=@n1.Rotation) -> (ReturnValue:Vector)
   > Set AimTarget(AimTarget:Vector=vector + vector(A=@n1, B=vector * vector(A=@n2, B:int=10000) -> (ReturnValue=~))) -> (Output_Get:Vector)
   TryGetPawnOwner()
   Event BlueprintInitializeAnimation()
   Event BlueprintBeginPlay() {is=title K2Node_Event}
  ]}
 ItemAimOffset {nodes [
   Input Pose!() -> (Pose, AimOffsetAlpha:double, Yaw:double, Pitch:double) {name=InPose_AimOffset}
   Input Pose!() -> (Pose) {name=InPose_PreAimOffsets}
   Output Pose() {layerGroup=Item Override}
  ]}
 ItemFallAdditive {nodes [
   Input Pose() {name=InPoseItemFallAdditive}
   Output Pose() {layerGroup=Item Override}
  ]}
 ItemFallLandAdditive {nodes [
   Input Pose() {name=InPoseItemFallLandAdditive}
   Output Pose() {layerGroup=Item Override}
  ]}
 ItemFallLevel2Override {nodes [
   Input Pose() {name=InPose_FallLevel2}
   Output Pose() {layerGroup=Item Override}
  ]}
 ItemFinalPoseOverride {nodes [
   n1 = Input Pose() {name=InFinalPose}
   n2 = Look At - Bone: hand_r!(
    ComponentPose=Look At - Bone: head!(ComponentPose=Local To Component(LocalPose=@n1), Alpha=1, LookAtLocation:Vector="100,0,0") -> (Pose) {boneToModify {boneName=head}, lookAtTarget=~}
    Alpha=1
    LookAtLocation:Vector="100,0,0"
   ) -> (Pose) {interpolationType=Sinusoidal, boneToModify {boneName=hand_r}, lookAt_Axis {axis {x=1, z=-1}}, lookUp_Axis {axis {y=1, z=-1}}, lookAtTarget=~}
   Output Pose(Result=Apply Additive(Base=Component To Local(ComponentPose=@n2))) {layerGroup=Item Override}
  ]}
 ItemFullBodyAimOffsets {nodes [
   Input Pose!() -> (Pose, FullBody_AimOffsetAlpha:double, Aim_Yaw:double, Aim_Pitch:double) {name=InPoseFullBodyAimOffsets}
   Input Pose!() -> (Pose) {name=InPosePreFullBodyAimOffsets}
   Output Pose() {layerGroup=Item Override}
  ]}
 ItemFullBodyOverride {nodes [
   Input Pose() {name=InPoseFullBody}
   Output Pose() {layerGroup=Item Override}
  ]}
 ItemIdleOverride {nodes [
   Input Pose() {name=InPose_Idle}
   Output Pose() {layerGroup=Item Override}
  ]}
 ItemJetPackJumpAdditive {nodes [Output Pose() {layerGroup=Item Override}]}
 ItemJetPackStartAdditive {nodes [Output Pose() {layerGroup=Item Override}]}
 ItemJumpLoopAdditive {nodes [
   Input Pose() {name=InPoseItemJumpLoopAdditive}
   Output Pose() {layerGroup=Item Override}
  ]}
 ItemJumpOverride {nodes [
   Input Pose() {name=InPoseJump}
   Output Pose() {layerGroup=Item Override}
  ]}
 ItemJumpUpAdditive {nodes [
   Input Pose() {name=InPoseItemJumpUpAdditive}
   Output Pose() {layerGroup=Item Override}
  ]}
 ItemPreAimOffsets {nodes [
   Input Pose() {name=InPosePreAimOffsets}
   Output Pose() {layerGroup=Item Override}
  ]}
 ItemPreIK {nodes [
   Input Pose() {name=InPose_PreIK}
   Output Pose() {layerGroup=Item Override}
  ]}
 ItemRightHandIK {nodes [
   Input Pose() {name=InPosePreRightHandIK}
   Input Pose() {name=InPosePostRightHandIK}
   Output Pose() {layerGroup=Item Override}
  ]}
 ItemSkeletalControl {nodes [
   Input Pose()
   Output Pose() {layerGroup=Item Override}
  ]}
 ItemSwimJumpFallLoop {nodes [Output Pose() {layerGroup=Item Override}]}
 ItemSwimJumpSurfaceEnd {nodes [Output Pose() {layerGroup=Item Override}]}
 ItemSwimJumpSurfaceLoop {nodes [Output Pose() {layerGroup=Item Override}]}
 ItemUpperBody {nodes [
   Input Pose!() -> (Pose, InputParam:Fort Anim Input Adjusted Aim Structure) {name=InPoseUpperBody}
   Output Pose() {layerGroup=Item Override}
  ]}
 ItemZipLineStartAdditive {nodes [Output Pose() {layerGroup=Item Override}]}
]
variables [
 AimTarget {type=Vector}
 OwnerPawn {type=Pawn, default=null}
]
